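# Sound cue for a looping medium interior ambience bed, built on wood-interior recordings and routed to the indoor ambient sound class.
# A quality-level node selects between two looping wave players, using the quad wave for the first two levels and the stereo wave for the last two.
# The cue plays at reduced volume and references a shared ambient attenuation asset.
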SoundCue /Game/Sounds/Apollo/Biomes/Interiors_Generic/AMB_Bed_Int_Med_Loop_Cue
root=QualityLevel_0
nodes [
 {id=QualityLevel_0, children [WavePlayer_2,WavePlayer_2,WavePlayer_3,WavePlayer_3]}
 {id=WavePlayer_2, properties {bLooping=true, soundWaveAssetPtr=&/AMB_Bed_Int_Wood_Med_01_Quad}}
 {id=WavePlayer_3, properties {bLooping=true, soundWaveAssetPtr=&/AMB_Bed_Int_Wood_Med_01_St}}
]
properties {
 attenuationSettings=&Fort_Ambient_Athena_Att
 maxDistance=2097152
 soundClassObject=&Fort_Ambient_Indoors_Athena
 volumeMultiplier=0.401
 attenuationOverrides {bAttenuate=false}
}
