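# Sound cue that plays the looping ambient bed for small metal interiors.
# A quality-level node feeds the first two of its four slots from one shared looping wave player using a quad-named wave, leaving the other two empty.
# It uses the Athena indoor ambience sound class and ambient attenuation asset at reduced volume.
# The cue also carries an attenuation override block with distance attenuation off.
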SoundCue /Game/Sounds/Apollo/Biomes/Interiors_Generic/AMB_Bed_Int_Metal_Sm_Loop_Cue
root=QualityLevel_0
nodes [
 {id=QualityLevel_0, children [WavePlayer_4,WavePlayer_4,null,null]}
 {id=WavePlayer_4, properties {bLooping=true, soundWaveAssetPtr=&/AMB_Bed_Int_Metal_Small_01_Quad}}
]
properties {
 attenuationSettings=&Fort_Ambient_Athena_Att
 maxDistance=2097152
 soundClassObject=&Fort_Ambient_Indoors_Athena
 volumeMultiplier=0.708
 attenuationOverrides {bAttenuate=false}
}
