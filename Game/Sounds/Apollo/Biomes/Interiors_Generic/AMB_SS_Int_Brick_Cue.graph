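# Sound cue for a one-shot brick ambience sound in generic interiors, classed as indoor ambient.
# A Random node picks one of fourteen equally weighted brick wave variations.
# A Modulator above it randomly lowers pitch and volume slightly on each play to vary repeats.
# Playback runs at a reduced overall volume through the shared tile-ambient one-shot attenuation.
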SoundCue /Game/Sounds/Apollo/Biomes/Interiors_Generic/AMB_SS_Int_Brick_Cue
root=Modulator_0
nodes [
 {id=Modulator_0, children [Random_0], properties {pitchMax=1, pitchMin=0.8, volumeMax=1, volumeMin=0.7}}
 {
  id=Random_0
  children [
   WavePlayer_0
   WavePlayer_1
   WavePlayer_2
   WavePlayer_11
   WavePlayer_12
   WavePlayer_13
   WavePlayer_14
   WavePlayer_15
   WavePlayer_16
   WavePlayer_17
   WavePlayer_18
   WavePlayer_19
   WavePlayer_20
   WavePlayer_21
  ]
  properties {weights [1,1,1,1,1,1,1,1,1,1,1,1,1,1]}
 }
 {id=WavePlayer_0, properties {soundWaveAssetPtr=&/AMB_SS_Int_Brick_10}}
 {id=WavePlayer_1, properties {soundWaveAssetPtr=&/AMB_SS_Int_Brick_11}}
 {id=WavePlayer_11, properties {soundWaveAssetPtr=&/AMB_SS_Int_Brick_13}}
 {id=WavePlayer_12, properties {soundWaveAssetPtr=&/AMB_SS_Int_Brick_14}}
 {id=WavePlayer_13, properties {soundWaveAssetPtr=&/AMB_SS_Int_Brick_01}}
 {id=WavePlayer_14, properties {soundWaveAssetPtr=&/AMB_SS_Int_Brick_02}}
 {id=WavePlayer_15, properties {soundWaveAssetPtr=&/AMB_SS_Int_Brick_03}}
 {id=WavePlayer_16, properties {soundWaveAssetPtr=&/AMB_SS_Int_Brick_04}}
 {id=WavePlayer_17, properties {soundWaveAssetPtr=&/AMB_SS_Int_Brick_05}}
 {id=WavePlayer_18, properties {soundWaveAssetPtr=&/AMB_SS_Int_Brick_06}}
 {id=WavePlayer_19, properties {soundWaveAssetPtr=&/AMB_SS_Int_Brick_07}}
 {id=WavePlayer_2, properties {soundWaveAssetPtr=&/AMB_SS_Int_Brick_12}}
 {id=WavePlayer_20, properties {soundWaveAssetPtr=&/AMB_SS_Int_Brick_08}}
 {id=WavePlayer_21, properties {soundWaveAssetPtr=&/AMB_SS_Int_Brick_09}}
]
properties {
 attenuationSettings=/Game/Sounds/Attenuation/Fort_TileAmbient_OneShot_Att
 maxDistance=8000
 soundClassObject=&Fort_Ambient_Indoors_Athena
 volumeMultiplier=0.354
}
