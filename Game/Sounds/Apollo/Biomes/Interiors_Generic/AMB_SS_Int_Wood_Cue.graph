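# Sound cue for one-shot ambient wood sounds in generic interiors of the Apollo biomes.
# A Random node picks among ten equally weighted wood wave variations, and a Modulator varies pitch downward and volume slightly on each play.
# Output is scaled down in volume, routed to the indoor ambient sound class, and spatialized with the tile-ambient one-shot attenuation.
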SoundCue /Game/Sounds/Apollo/Biomes/Interiors_Generic/AMB_SS_Int_Wood_Cue
root=Modulator_0
nodes [
 {id=Modulator_0, children [Random_0], properties {pitchMax=1, pitchMin=0.8, volumeMax=1, volumeMin=0.7}}
 {
  id=Random_0
  children [
   WavePlayer_0
   WavePlayer_1
   WavePlayer_2
   WavePlayer_11
   WavePlayer_12
   WavePlayer_13
   WavePlayer_14
   WavePlayer_15
   WavePlayer_16
   WavePlayer_17
  ]
  properties {weights [1,1,1,1,1,1,1,1,1,1]}
 }
 {id=WavePlayer_0, properties {soundWaveAssetPtr=&/AMB_SS_Int_Wood_10}}
 {id=WavePlayer_1, properties {soundWaveAssetPtr=&/AMB_SS_Int_Wood_01}}
 {id=WavePlayer_11, properties {soundWaveAssetPtr=&/AMB_SS_Int_Wood_03}}
 {id=WavePlayer_12, properties {soundWaveAssetPtr=&/AMB_SS_Int_Wood_04}}
 {id=WavePlayer_13, properties {soundWaveAssetPtr=&/AMB_SS_Int_Wood_05}}
 {id=WavePlayer_14, properties {soundWaveAssetPtr=&/AMB_SS_Int_Wood_06}}
 {id=WavePlayer_15, properties {soundWaveAssetPtr=&/AMB_SS_Int_Wood_07}}
 {id=WavePlayer_16, properties {soundWaveAssetPtr=&/AMB_SS_Int_Wood_08}}
 {id=WavePlayer_17, properties {soundWaveAssetPtr=&/AMB_SS_Int_Wood_09}}
 {id=WavePlayer_2, properties {soundWaveAssetPtr=&/AMB_SS_Int_Wood_02}}
]
properties {
 attenuationSettings=/Game/Sounds/Attenuation/Fort_TileAmbient_OneShot_Att
 maxDistance=8000
 soundClassObject=&Fort_Ambient_Indoors_Athena
 volumeMultiplier=0.354
}
